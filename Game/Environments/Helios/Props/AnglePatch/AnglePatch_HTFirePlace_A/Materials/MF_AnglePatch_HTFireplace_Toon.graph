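# Material function for the HT fireplace prop that outputs an unlit-looking toon heart, with black base color and an emissive blend between line and heart colors.
# A contrast-sharpened SDF from the diffuse red channel is capped by a hard cel step from normal-map dot camera vector.
# Time drives two multiplied triangle waves as Flicker for animated brightness.
# The blend mask is scaled by inverse eye adaptation before driving the color lerp.
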MaterialFunction /Game/Environments/Helios/Props/AnglePatch/AnglePatch_HTFirePlace_A/Materials/MF_AnglePatch_HTFireplace_Toon
graph {nodes [
  n1 = TextureSampleParameter2D() {
   group=Toon Heart
   parameterName=Normal
   samplerType=SAMPLERTYPE_Normal
   texture=/Game/Environments/Helios/Props/AnglePatch/AnglePatch_HTFirePlace_A/Textures/T_AnglePatch_HTFireplace_A_B_N
  }
  n2 = Transform(None=@n1.RGB)
  Add(A=@n2, B=Constant3Vector() {constant=<rgba 0 0 -0.3 1>})
  n3 = Constant()
  n4 = VectorParameter() {group=Toon Heart, parameterName=Color_Lines, sortPriority=0, defaultValue=<rgba 0 0 0 1>}
  n5 = VectorParameter() {group=Toon Heart, parameterName=Color_Heart, sortPriority=0, defaultValue=<rgba 2 0 0 1>}
  n6 = TextureSampleParameter2D() {
   group=Toon Heart
   parameterName=Diffuse
   texture=/Game/Environments/Helios/Props/AnglePatch/AnglePatch_HTFirePlace_A/Textures/T_AnglePatch_HTFireplace_A_B_D
  }
  n7 = MaterialFunctionCall(
   In=Subtract(A=@n6.R, B=ScalarParameter() {defaultValue=0.13, group=Toon Heart, parameterName=SDF_Lines_Thickness, sortPriority=5})
   Contrast=ScalarParameter() {defaultValue=0.07, group=Toon Heart, parameterName=SDF_Lines_Contrast, sortPriority=5}
  ) {materialFunction=&CheapContrast}
  n8 = NamedRerouteUsage()
  n9 = Multiply(A=@n8, B=ScalarParameter() {defaultValue=0.1, parameterName=Fresnel_flicker}) {constB=0.05}
  n10 = ScalarParameter() {defaultValue=0.25, group=Toon Heart, parameterName=Fresnel Opacity, sortPriority=10}
  n11 = ScalarParameter() {defaultValue=0.8, group=Toon Heart, parameterName=Flicker Min}
  n12 = ScalarParameter() {defaultValue=1, group=Toon Heart, parameterName=Flicker Max}
  n13 = Add(A=DotProduct(A=@n2, B=CameraVectorWS()), B=ScalarParameter() {defaultValue=-0.3, group=Toon Heart, parameterName=CellShadingStep_Offset})
  n14 = Add(A=@n13, B=Multiply(A=NamedRerouteUsage()) {constB=-0.05})
  n15 = ScalarParameter() {defaultValue=10, parameterName=Fresnel Hardness}
  n16 = Min(
   A=@n7.Result
   B=LinearInterpolate(
    A=Add(A=@n9, B=@n10)
    B=LinearInterpolate(A=@n11, B=@n12, Alpha=@n8)
    Alpha=MaterialFunctionCall(In=@n14, Contrast=@n15) {materialFunction=&CheapContrast}.Result
   )
  )
  n17 = NamedRerouteDeclaration(None=Multiply(A=Saturate(None=@n16), B=EyeAdaptationInverse().EyeAdaptationInverse)) {name=SDF, nodeColor=<rgba 1 0.3516 0 1>}
  FunctionOutput(None=MakeMaterialAttributes!(
    BaseColor=@n3
    Metallic=@n3
    Specular=@n3
    Roughness=Constant() {r=1}
    Anisotropy
    EmissiveColor=LinearInterpolate(A=@n4.RGB, B=@n5.RGB, Alpha=@n17)
    Opacity
    OpacityMask
    Normal
    Tangent
    WorldPositionOffset
    SubsurfaceColor
    ClearCoat
    ClearCoatRoughness
    AmbientOcclusion
    Refraction
    CustomizedUVs_0
    CustomizedUVs_1
    CustomizedUVs_2
    CustomizedUVs_3
    CustomizedUVs_4
    CustomizedUVs_5
    CustomizedUVs_6
    CustomizedUVs_7
    PixelDepthOffset
    ShadingModel
    Displacement
   )) {outputName=BlendMaterialAttributes}
  Min(A=Add(A=ComponentMask(None=@n2) {b=true}) {constB=0.75}, B=@n13)
  n18 = Multiply(A=Time(), B=ScalarParameter() {defaultValue=3, group=Toon Heart, parameterName=Flicker_Speed})
  n19 = Multiply(A=Multiply(A=@n18) {constB=0.37})
  n20 = Subtract(A=Multiply(A=Frac(None=@n19)) {constB=2})
  n21 = Subtract(A=Multiply(A=Frac(None=@n18)) {constB=2})
  NamedRerouteDeclaration(None=Multiply(A=Abs(None=@n20), B=Abs(None=@n21))) {name=Flicker, nodeColor=<rgba 0 1 0.7656 1>}
  SmoothStep(Min=Constant() {r=0.48}, Max=Constant() {r=0.52}, Value=@n14)
  Subtract(A=@n7.Result)
  VertexNormalWS()
 ]}
