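# Masked, unlit, two-sided inverted-hull outline material for the AnglePatch HT fireplace prop, with a default black outline color.
# The prop's normal map is adjusted by FlattenNormal, normalized and transformed to form the hull offset direction, with flatness tunable for separated geometry.
# An FOV-compensated distance fade blends between near and far thickness to scale that offset.
# An inverted TwoSidedSign distinguishes back faces from front faces for the hull.
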Material /Game/Environments/Helios/Props/AnglePatch/AnglePatch_HTFirePlace_A/Materials/M_AnglePatch_Outline
graph {
 properties {bAutomaticallySetUsageInEditor=false, blendMode=BLEND_Masked, shadingModel=MSM_Unlit, twoSided=true, usedShadingModels=Unlit}
 nodes [
  n1 = TextureSampleParameter2D() {
   group=01_InvertedHull_Outlines
   parameterName=Normal
   samplerType=SAMPLERTYPE_Normal
   texture=/Game/Environments/Helios/Props/AnglePatch/AnglePatch_HTFirePlace_A/Textures/T_AnglePatch_HTFireplace_A_B_N
  }
  n2 = ScalarParameter() {
   defaultValue=-0.4
   desc="(0.4)\r\nvalues above 1.0 and below 0.0 can help with seperated geometry.\r\n-0.1 seems like a good well you to alight split geo."
   group=01_InvertedHull_Outlines
   parameterName=NormalMap_Flatness
   sliderMax=2
   sliderMin=-1
   sortPriority=5
  }
  n3 = MaterialFunctionCall(Normal=@n1.RGB, Flatness=@n2) {materialFunction=&FlattenNormal}
  n4 = ScalarParameter() {defaultValue=1000, group=01_InvertedHull_Outlines, parameterName=Fade_Distance, sliderMax=1500, sliderMin=500, sortPriority=5}
  n5 = MaterialFunctionCall!(
   Fade Offset=ScalarParameter() {defaultValue=50, group=01_InvertedHull_Outlines, parameterName=Fade_Offset, sliderMax=100, sortPriority=5}
   Fade Distance=@n4
  ) {materialFunction=/Game/Global/Materials/Character_Parent_2023/Anime_Parent_2023/MaterialFunctions/MFC_DistanceFade_FOV_compensation}
  Multiply(
   A=Transform(None=Normalize(VectorInput=@n3.Result))
   B=LinearInterpolate(
    A=ScalarParameter() {defaultValue=0.2, desc=0.35, parameterName=Outline_Thickness}
    B=ScalarParameter() {defaultValue=3, desc=3, parameterName=Outline_Thickness_Far}
    Alpha=NamedRerouteDeclaration(None=@n5.Result) {name=DistanceFading, nodeColor=<rgba 1 0.3047 0 1>}
   )
  )
  OneMinus(None=TwoSidedSign())
  VectorParameter() {parameterName=Outline_Color, defaultValue=<rgba 0 0 0 1>}
  VertexNormalWS()
 ]
}
